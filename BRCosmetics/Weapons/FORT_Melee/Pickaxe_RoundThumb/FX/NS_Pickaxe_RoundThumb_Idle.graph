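# Niagara idle effect for the RoundThumb pickaxe that swirls small mesh pieces, named as metal, around the weapon.
# One emitter spawns short-lived mesh particles intermittently, with random mesh index and a size biased small, and they tumble while scaling up then down.
# Each update samples the system's skeletal mesh with an offset and computes an orbit around that point.
# InterpolateOverTime smooths the orbit target into particle position, so pieces trail the moving weapon.
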NiagaraSystem /BRCosmetics/Weapons/FORT_Melee/Pickaxe_RoundThumb/FX/NS_Pickaxe_RoundThumb_Idle
dependencies {
 usedDataInterfaces [
  &NiagaraDataInterfaceCurve
  &NiagaraDataInterfaceMeshRendererInfo
  &NiagaraDataInterfaceSkeletalMesh
 ]
 usedDynamicInputs []
 usedModules [
  &self:SystemSpawnScript.SystemScriptSource.SystemScriptGraph.NiagaraNodeAssignment_0.SetVariables_#1
  &SystemState
 ]
}
emitters {Metal_Swirl {
  emitterUpdateScript [
   &EmitterState(Life Cycle Mode=::NewEnumerator1, Loop Behavior=::NewEnumerator1, Loop Duration Mode=::NewEnumerator1, UseLoopDelay=false)
   &SpawnRate(Spawn Probability=0.5, SpawnRate=10, Use Spawn Probability=true) {module=SpawnRate001}
  ]
  particleSpawnScript [&InitializeParticle(
    Color=<rgba 0.4323 0.3701 0.3377 1>
    Lifetime=1.5
    Mesh Renderer Array Visibility Mode=::NewEnumerator1
    Mesh Renderer Info=DataInterface() {propertyValues {MeshRenderer=&self:Metal_Swirl_0.NiagaraMeshRendererProperties_0}}
    Mesh Renderer Mesh Index:NiagaraInt32=&RandomRangeInt(Minimum=0, Maximum=9)
    Mesh Scale Mode=::NewEnumerator3
    Mesh Uniform Scale=&Lerp_Float(
     A=0.3
     B=0.8
     Alpha=&FloatFromCurve(
      FloatCurve=DataInterface() {propertyValues {Curve {keys [
          {interpMode=RCIM_Cubic, tangentMode=RCTM_User, time=0.5, value=0}
          {arriveTangent=3.2165, interpMode=RCIM_Cubic, leaveTangent=3.2165, tangentMode=RCTM_User, time=1, value=1}
         ]}}}
      CurveIndex=&RandomRangeFloat(Minimum=0, Maximum=1)
     )
    )
    UsePositionOffset=false
    Write Material Random=true
   )]
  particleUpdateScript [
   &ParticleState()
   &UpdateMeshOrientation(
    Rotation Rate=&RandomRangeFloat(Minimum=0.1, Maximum=0.5)
    Rotation Vector=&RandomVector(Vector Scale=1)
   )
   &ScaleMeshSize(Scale Factor=&VectorFromFloat(Value=&FloatFromCurve(FloatCurve=DataInterface() {propertyValues {Curve {keys [
          {arriveTangent=3.1214, interpMode=RCIM_Cubic, leaveTangent=3.1214, tangentMode=RCTM_User, time=0, value=0}
          {interpMode=RCIM_Cubic, tangentMode=RCTM_User, time=0.5, value=1}
          {tangentMode=RCTM_User, time=1, value=0}
         ]}}})))
   &SkeletalMeshLocation(
    Bone Sampling Mode=::NewEnumerator4
    MaskSpawnGroup=false
    Orientation Sampling=::NewEnumerator1
    Position Sampling=::NewEnumerator2
    Sampled Position Offset=&MakeVector(X=&RandomRangeFloat(Minimum=-6, Maximum=0), Y=-2, Z=0)
    Skeletal Mesh=$System.Skeletal Mesh
   )
   &RotateAroundPoint(
    Radius:NiagaraFloat=&FloatFromCurve(
     FloatCurve=DataInterface() {propertyValues {Curve {keys [
         {arriveTangent=0.0021, interpMode=RCIM_Cubic, leaveTangent=0.0021, tangentMode=RCTM_User, time=0, value=0.8}
         {interpMode=RCIM_Cubic, tangentMode=RCTM_User, time=0.5, value=1}
         {tangentMode=RCTM_User, time=1, value=0.8}
        ]}}}
     Scale Curve=&RandomRangeFloat(Minimum=15, Maximum=20)
    )
    Rotation Center:NiagaraPosition=$Particles.SkeletalMeshLocation.SampledPosition
    Rotation Phase:NiagaraFloat=&RandomRangeFloat(Minimum=0, Maximum=1)
    Rotation Rate:NiagaraFloat=&FloatFromCurve(
     FloatCurve=DataInterface() {propertyValues {Curve {keys [{interpMode=RCIM_Cubic, tangentMode=RCTM_User, time=0, value=1}, {interpMode=RCIM_Cubic, tangentMode=RCTM_User, time=1, value=0.2}]}}}
     Scale Curve=&RandomRangeFloat(Minimum=0.2, Maximum=0.4)
    )
    Set Particle Position=false
    X Axis:Vector3f=&TransformVector(
     Vector=&MakeVector(X=&RandomRangeFloat(Minimum=-2, Maximum=2), Y=1, Z=0)
     Destination Space=::Simulation
    )
    Y Axis:Vector3f=&TransformVector(Vector=&MakeVector(X=0, Y=0, Z=1), Destination Space=::Simulation)
   ) {module=RotateAroundPoint001}
   &InterpolateOverTime(
    Rate Of Change:NiagaraFloat=&RandomRangeFloat(Minimum=10, Maximum=20)
    Target Vector Value:Vector3f=$Output.RotateAroundPoint001.OutputPosition
    Variable Type=::NewEnumerator2
   )
   /SetVariables(Particles.Position=$Particles.InterpolateOverTime.InterpolateOverTime_V3_.Moving Average) {isSetParameters=true}
  ]
  rendererClasses [&NiagaraMeshRendererProperties]
  renderers [{rendererClass=&NiagaraMeshRendererProperties, rendererIndex=0}]
 }}
summary {
 systemName=NS_Pickaxe_RoundThumb_Idle
 emitters [{emitterName=Metal_Swirl, rendererClasses [&NiagaraMeshRendererProperties]}]
 userVariables []
}
