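# Physics asset for the RoundThumb pickaxe mesh, giving four wobble pieces on the axe bodies for physics-driven motion.
# Each piece pairs a small capsule on a base joint with a larger, offset capsule on a dyn joint.
# One constraint links each pair, with twist and first swing limits at zero, so the piece wobbles in a single swing plane within a limited arc.
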PhysicsAsset /BRCosmetics/Weapons/FORT_Melee/Pickaxe_RoundThumb/Meshes/RoundThumb_Axe_Physics
bodies {
 base_wobblePiece_0_jnt [{shapeType=Capsule, length=0.505, radius=0.505, shapeName=base_wobblePiece_0_jnt_capsule, rotation=<rot 0 -90 0>}]
 base_wobblePiece_1_jnt [{shapeType=Capsule, length=0.505, radius=0.505, shapeName=base_wobblePiece_1_jnt_capsule, rotation=<rot 0 -90 0>}]
 base_wobblePiece_2_jnt [{shapeType=Capsule, length=0.505, radius=0.505, shapeName=base_wobblePiece_2_jnt_capsule, rotation=<rot 0 -90 0>}]
 base_wobblePiece_3_jnt [{shapeType=Capsule, length=0.505, radius=0.505, shapeName=base_wobblePiece_3_jnt_capsule, rotation=<rot 0 -90 0>}]
 dyn_wobblePiece_0_jnt [{shapeType=Capsule, length=2, radius=1.8, shapeName=dyn_wobblePiece_0_jnt_capsule, center=<5 0 -1>, rotation=<rot 90 0 0>}]
 dyn_wobblePiece_1_jnt [{shapeType=Capsule, length=2, radius=1.8, shapeName=dyn_wobblePiece_1_jnt_capsule, center=<5 0 -1>, rotation=<rot 90 0 0>}]
 dyn_wobblePiece_2_jnt [{shapeType=Capsule, length=2, radius=1.8, shapeName=dyn_wobblePiece_2_jnt_capsule, center=<5 0 -1>, rotation=<rot 90 0 0>}]
 dyn_wobblePiece_3_jnt [{shapeType=Capsule, length=2, radius=1.8, shapeName=dyn_wobblePiece_3_jnt_capsule, center=<5 0 -1>, rotation=<rot 90 0 0>}]
}
constraints [
 {bone1Name=dyn_wobblePiece_2_jnt, bone2Name=base_wobblePiece_2_jnt, swing1LimitDegrees=0, swing2LimitDegrees=40, twistLimitDegrees=0}
 {bone1Name=dyn_wobblePiece_0_jnt, bone2Name=base_wobblePiece_0_jnt, swing1LimitDegrees=0, swing2LimitDegrees=40, twistLimitDegrees=0}
 {bone1Name=dyn_wobblePiece_1_jnt, bone2Name=base_wobblePiece_1_jnt, swing1LimitDegrees=0, swing2LimitDegrees=40, twistLimitDegrees=0}
 {bone1Name=dyn_wobblePiece_3_jnt, bone2Name=base_wobblePiece_3_jnt, swing1LimitDegrees=0, swing2LimitDegrees=40, twistLimitDegrees=0}
]
