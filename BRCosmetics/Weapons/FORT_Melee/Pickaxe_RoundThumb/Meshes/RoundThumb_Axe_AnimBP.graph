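# Animation Blueprint for the RoundThumb pickaxe that adds procedural floating motion and light physics to its loose pieces.
# Each update, sine-wave macros compute per-piece offset vectors from game time, and a deprecated helper sets GravityOverride.
# The AnimGraph chains additive bone-space translation nodes for 39 piece bones through cached poses.
# A low-alpha RigidBody node, simulated relative to the two-handed handle bone, then feeds DefaultSlot.
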AnimBlueprint /BRCosmetics/Weapons/FORT_Melee/Pickaxe_RoundThumb/Meshes/RoundThumb_Axe_AnimBP
target_skeleton=&/RoundThumb_Axe_Skeleton
graphs [
 AnimGraph {nodes [
   Output Pose(Result=Slot 'DefaultSlot'(Source=Use cached pose 'RigidBodyCache'()))
   n1 = "Transform (Modify) Bone - Bone: dyn_piece_36_jnt"!(ComponentPose=Local To Component(LocalPose=Use cached pose 'AniPieces_Grp6Cache'()), Alpha=1, Translation="0,0,0") -> (Pose) {translationMode=BMM_Additive, translationSpace=BCS_BoneSpace, alphaBoolBlend {blendOption=Linear}, boneToModify {boneName=dyn_piece_36_jnt}}
   n2 = "Transform (Modify) Bone - Bone: dyn_piece_37_jnt"!(ComponentPose=@n1, Alpha=1, Translation="0,0,0") -> (Pose) {translationMode=BMM_Additive, translationSpace=BCS_BoneSpace, alphaBoolBlend {blendOption=Linear}, boneToModify {boneName=dyn_piece_37_jnt}}
   n3 = "Transform (Modify) Bone - Bone: dyn_piece_38_jnt"!(ComponentPose=@n2, Alpha=1, Translation="0,0,0") -> (Pose) {translationMode=BMM_Additive, translationSpace=BCS_BoneSpace, alphaBoolBlend {blendOption=Linear}, boneToModify {boneName=dyn_piece_38_jnt}}
   Save cached pose 'OutputAniPiecesCache'(Pose=Component To Local(ComponentPose=@n3)) {cacheName=OutputAniPiecesCache}
   Save cached pose 'InputPoseCache'(Pose=Input Pose()) {cacheName=InputPoseCache}
   n4 = "Transform (Modify) Bone - Bone: dyn_piece_0_jnt"!(ComponentPose=Local To Component(LocalPose=Use cached pose 'InputPoseCache'()), Alpha=1, Translation="0,0,0") -> (Pose) {translationMode=BMM_Additive, translationSpace=BCS_BoneSpace, alphaBoolBlend {blendOption=Linear}, boneToModify {boneName=dyn_piece_0_jnt}}
   n5 = "Transform (Modify) Bone - Bone: dyn_piece_1_jnt"!(ComponentPose=@n4, Alpha=1, Translation="0,0,0") -> (Pose) {translationMode=BMM_Additive, translationSpace=BCS_BoneSpace, alphaBoolBlend {blendOption=Linear}, boneToModify {boneName=dyn_piece_1_jnt}}
   n6 = "Transform (Modify) Bone - Bone: dyn_piece_2_jnt"!(ComponentPose=@n5, Alpha=1, Translation="0,0,0") -> (Pose) {translationMode=BMM_Additive, translationSpace=BCS_BoneSpace, alphaBoolBlend {blendOption=Linear}, boneToModify {boneName=dyn_piece_2_jnt}}
   n7 = "Transform (Modify) Bone - Bone: dyn_piece_3_jnt"!(ComponentPose=@n6, Alpha=1, Translation="0,0,0") -> (Pose) {translationMode=BMM_Additive, translationSpace=BCS_BoneSpace, alphaBoolBlend {blendOption=Linear}, boneToModify {boneName=dyn_piece_3_jnt}}
   n8 = "Transform (Modify) Bone - Bone: dyn_piece_4_jnt"!(ComponentPose=@n7, Alpha=1, Translation="0,0,0") -> (Pose) {translationMode=BMM_Additive, translationSpace=BCS_BoneSpace, alphaBoolBlend {blendOption=Linear}, boneToModify {boneName=dyn_piece_4_jnt}}
   n9 = "Transform (Modify) Bone - Bone: dyn_piece_5_jnt"!(ComponentPose=@n8, Alpha=1, Translation="0,0,0") -> (Pose) {translationMode=BMM_Additive, translationSpace=BCS_BoneSpace, alphaBoolBlend {blendOption=Linear}, boneToModify {boneName=dyn_piece_5_jnt}}
   Save cached pose 'AniPieces_Grp1Cache'(Pose=Component To Local(ComponentPose=@n9)) {cacheName=AniPieces_Grp1Cache}
   n10 = "Transform (Modify) Bone - Bone: dyn_piece_6_jnt"!(ComponentPose=Local To Component(LocalPose=Use cached pose 'AniPieces_Grp1Cache'()), Alpha=1, Translation="0,0,0") -> (Pose) {translationMode=BMM_Additive, translationSpace=BCS_BoneSpace, alphaBoolBlend {blendOption=Linear}, boneToModify {boneName=dyn_piece_6_jnt}}
   n11 = "Transform (Modify) Bone - Bone: dyn_piece_7_jnt"!(ComponentPose=@n10, Alpha=1, Translation="0,0,0") -> (Pose) {translationMode=BMM_Additive, translationSpace=BCS_BoneSpace, alphaBoolBlend {blendOption=Linear}, boneToModify {boneName=dyn_piece_7_jnt}}
   n12 = "Transform (Modify) Bone - Bone: dyn_piece_8_jnt"!(ComponentPose=@n11, Alpha=1, Translation="0,0,0") -> (Pose) {translationMode=BMM_Additive, translationSpace=BCS_BoneSpace, alphaBoolBlend {blendOption=Linear}, boneToModify {boneName=dyn_piece_8_jnt}}
   n13 = "Transform (Modify) Bone - Bone: dyn_piece_9_jnt"!(ComponentPose=@n12, Alpha=1, Translation="0,0,0") -> (Pose) {translationMode=BMM_Additive, translationSpace=BCS_BoneSpace, alphaBoolBlend {blendOption=Linear}, boneToModify {boneName=dyn_piece_9_jnt}}
   n14 = "Transform (Modify) Bone - Bone: dyn_piece_10_jnt"!(ComponentPose=@n13, Alpha=1, Translation="0,0,0") -> (Pose) {translationMode=BMM_Additive, translationSpace=BCS_BoneSpace, alphaBoolBlend {blendOption=Linear}, boneToModify {boneName=dyn_piece_10_jnt}}
   n15 = "Transform (Modify) Bone - Bone: dyn_piece_11_jnt"!(ComponentPose=@n14, Alpha=1, Translation="0,0,0") -> (Pose) {translationMode=BMM_Additive, translationSpace=BCS_BoneSpace, alphaBoolBlend {blendOption=Linear}, boneToModify {boneName=dyn_piece_11_jnt}}
   Save cached pose 'AniPieces_Grp2Cache'(Pose=Component To Local(ComponentPose=@n15)) {cacheName=AniPieces_Grp2Cache}
   n16 = "Transform (Modify) Bone - Bone: dyn_piece_12_jnt"!(ComponentPose=Local To Component(LocalPose=Use cached pose 'AniPieces_Grp2Cache'()), Alpha=1, Translation="0,0,0") -> (Pose) {translationMode=BMM_Additive, translationSpace=BCS_BoneSpace, alphaBoolBlend {blendOption=Linear}, boneToModify {boneName=dyn_piece_12_jnt}}
   n17 = "Transform (Modify) Bone - Bone: dyn_piece_13_jnt"!(ComponentPose=@n16, Alpha=1, Translation="0,0,0") -> (Pose) {translationMode=BMM_Additive, translationSpace=BCS_BoneSpace, alphaBoolBlend {blendOption=Linear}, boneToModify {boneName=dyn_piece_13_jnt}}
   n18 = "Transform (Modify) Bone - Bone: dyn_piece_14_jnt"!(ComponentPose=@n17, Alpha=1, Translation="0,0,0") -> (Pose) {translationMode=BMM_Additive, translationSpace=BCS_BoneSpace, alphaBoolBlend {blendOption=Linear}, boneToModify {boneName=dyn_piece_14_jnt}}
   n19 = "Transform (Modify) Bone - Bone: dyn_piece_15_jnt"!(ComponentPose=@n18, Alpha=1, Translation="0,0,0") -> (Pose) {translationMode=BMM_Additive, translationSpace=BCS_BoneSpace, alphaBoolBlend {blendOption=Linear}, boneToModify {boneName=dyn_piece_15_jnt}}
   n20 = "Transform (Modify) Bone - Bone: dyn_piece_16_jnt"!(ComponentPose=@n19, Alpha=1, Translation="0,0,0") -> (Pose) {translationMode=BMM_Additive, translationSpace=BCS_BoneSpace, alphaBoolBlend {blendOption=Linear}, boneToModify {boneName=dyn_piece_16_jnt}}
   n21 = "Transform (Modify) Bone - Bone: dyn_piece_17_jnt"!(ComponentPose=@n20, Alpha=1, Translation="0,0,0") -> (Pose) {translationMode=BMM_Additive, translationSpace=BCS_BoneSpace, alphaBoolBlend {blendOption=Linear}, boneToModify {boneName=dyn_piece_17_jnt}}
   Save cached pose 'AniPieces_Grp3Cache'(Pose=Component To Local(ComponentPose=@n21)) {cacheName=AniPieces_Grp3Cache}
   n22 = "Transform (Modify) Bone - Bone: dyn_piece_18_jnt"!(ComponentPose=Local To Component(LocalPose=Use cached pose 'AniPieces_Grp3Cache'()), Alpha=1, Translation="0,0,0") -> (Pose) {translationMode=BMM_Additive, translationSpace=BCS_BoneSpace, alphaBoolBlend {blendOption=Linear}, boneToModify {boneName=dyn_piece_18_jnt}}
   n23 = "Transform (Modify) Bone - Bone: dyn_piece_19_jnt"!(ComponentPose=@n22, Alpha=1, Translation="0,0,0") -> (Pose) {translationMode=BMM_Additive, translationSpace=BCS_BoneSpace, alphaBoolBlend {blendOption=Linear}, boneToModify {boneName=dyn_piece_19_jnt}}
   n24 = "Transform (Modify) Bone - Bone: dyn_piece_20_jnt"!(ComponentPose=@n23, Alpha=1, Translation="0,0,0") -> (Pose) {translationMode=BMM_Additive, translationSpace=BCS_BoneSpace, alphaBoolBlend {blendOption=Linear}, boneToModify {boneName=dyn_piece_20_jnt}}
   n25 = "Transform (Modify) Bone - Bone: dyn_piece_21_jnt"!(ComponentPose=@n24, Alpha=1, Translation="0,0,0") -> (Pose) {translationMode=BMM_Additive, translationSpace=BCS_BoneSpace, alphaBoolBlend {blendOption=Linear}, boneToModify {boneName=dyn_piece_21_jnt}}
   n26 = "Transform (Modify) Bone - Bone: dyn_piece_22_jnt"!(ComponentPose=@n25, Alpha=1, Translation="0,0,0") -> (Pose) {translationMode=BMM_Additive, translationSpace=BCS_BoneSpace, alphaBoolBlend {blendOption=Linear}, boneToModify {boneName=dyn_piece_22_jnt}}
   n27 = "Transform (Modify) Bone - Bone: dyn_piece_23_jnt"!(ComponentPose=@n26, Alpha=1, Translation="0,0,0") -> (Pose) {translationMode=BMM_Additive, translationSpace=BCS_BoneSpace, alphaBoolBlend {blendOption=Linear}, boneToModify {boneName=dyn_piece_23_jnt}}
   Save cached pose 'AniPieces_Grp4Cache'(Pose=Component To Local(ComponentPose=@n27)) {cacheName=AniPieces_Grp4Cache}
   n28 = "Transform (Modify) Bone - Bone: dyn_piece_24_jnt"!(ComponentPose=Local To Component(LocalPose=Use cached pose 'AniPieces_Grp4Cache'()), Alpha=1, Translation="0,0,0") -> (Pose) {translationMode=BMM_Additive, translationSpace=BCS_BoneSpace, alphaBoolBlend {blendOption=Linear}, boneToModify {boneName=dyn_piece_24_jnt}}
   n29 = "Transform (Modify) Bone - Bone: dyn_piece_25_jnt"!(ComponentPose=@n28, Alpha=1, Translation="0,0,0") -> (Pose) {translationMode=BMM_Additive, translationSpace=BCS_BoneSpace, alphaBoolBlend {blendOption=Linear}, boneToModify {boneName=dyn_piece_25_jnt}}
   n30 = "Transform (Modify) Bone - Bone: dyn_piece_26_jnt"!(ComponentPose=@n29, Alpha=1, Translation="0,0,0") -> (Pose) {translationMode=BMM_Additive, translationSpace=BCS_BoneSpace, alphaBoolBlend {blendOption=Linear}, boneToModify {boneName=dyn_piece_26_jnt}}
   n31 = "Transform (Modify) Bone - Bone: dyn_piece_27_jnt"!(ComponentPose=@n30, Alpha=1, Translation="0,0,0") -> (Pose) {translationMode=BMM_Additive, translationSpace=BCS_BoneSpace, alphaBoolBlend {blendOption=Linear}, boneToModify {boneName=dyn_piece_27_jnt}}
   n32 = "Transform (Modify) Bone - Bone: dyn_piece_28_jnt"!(ComponentPose=@n31, Alpha=1, Translation="0,0,0") -> (Pose) {translationMode=BMM_Additive, translationSpace=BCS_BoneSpace, alphaBoolBlend {blendOption=Linear}, boneToModify {boneName=dyn_piece_28_jnt}}
   n33 = "Transform (Modify) Bone - Bone: dyn_piece_29_jnt"!(ComponentPose=@n32, Alpha=1, Translation="0,0,0") -> (Pose) {translationMode=BMM_Additive, translationSpace=BCS_BoneSpace, alphaBoolBlend {blendOption=Linear}, boneToModify {boneName=dyn_piece_29_jnt}}
   Save cached pose 'AniPieces_Grp5Cache'(Pose=Component To Local(ComponentPose=@n33)) {cacheName=AniPieces_Grp5Cache}
   n34 = RigidBody!(ComponentPose=Local To Component(LocalPose=Use cached pose 'OutputAniPiecesCache'()), Alpha=0.1, OverrideWorldGravity:Vector="0,0,0") -> (Pose) {
    bClampLinearTranslationLimitToRefPose=true
    bOverrideWorldGravity=true
    overridePhysicsAsset=&/RoundThumb_Axe_Physics
    simulationSpace=BaseBoneSpace
    alphaBoolBlend {blendOption=Linear}
    baseBoneRef {boneName=Melee_TwoHanded_Handle}
    componentLinearAccScale=<0.6 0.6 0.6>
    componentLinearVelScale=<0.6 0.6 0.6>
    overrideWorldGravity {z=-980}
   }
   Save cached pose 'RigidBodyCache'(Pose=Component To Local(ComponentPose=@n34)) {cacheName=RigidBodyCache}
   n35 = "Transform (Modify) Bone - Bone: dyn_piece_30_jnt"!(ComponentPose=Local To Component(LocalPose=Use cached pose 'AniPieces_Grp5Cache'()), Alpha=1, Translation="0,0,0") -> (Pose) {translationMode=BMM_Additive, translationSpace=BCS_BoneSpace, alphaBoolBlend {blendOption=Linear}, boneToModify {boneName=dyn_piece_30_jnt}}
   n36 = "Transform (Modify) Bone - Bone: dyn_piece_31_jnt"!(ComponentPose=@n35, Alpha=1, Translation="0,0,0") -> (Pose) {translationMode=BMM_Additive, translationSpace=BCS_BoneSpace, alphaBoolBlend {blendOption=Linear}, boneToModify {boneName=dyn_piece_31_jnt}}
   n37 = "Transform (Modify) Bone - Bone: dyn_piece_32_jnt"!(ComponentPose=@n36, Alpha=1, Translation="0,0,0") -> (Pose) {translationMode=BMM_Additive, translationSpace=BCS_BoneSpace, alphaBoolBlend {blendOption=Linear}, boneToModify {boneName=dyn_piece_32_jnt}}
   n38 = "Transform (Modify) Bone - Bone: dyn_piece_33_jnt"!(ComponentPose=@n37, Alpha=1, Translation="0,0,0") -> (Pose) {translationMode=BMM_Additive, translationSpace=BCS_BoneSpace, alphaBoolBlend {blendOption=Linear}, boneToModify {boneName=dyn_piece_33_jnt}}
   n39 = "Transform (Modify) Bone - Bone: dyn_piece_34_jnt"!(ComponentPose=@n38, Alpha=1, Translation="0,0,0") -> (Pose) {translationMode=BMM_Additive, translationSpace=BCS_BoneSpace, alphaBoolBlend {blendOption=Linear}, boneToModify {boneName=dyn_piece_34_jnt}}
   n40 = "Transform (Modify) Bone - Bone: dyn_piece_35_jnt"!(ComponentPose=@n39, Alpha=1, Translation="0,0,0") -> (Pose) {translationMode=BMM_Additive, translationSpace=BCS_BoneSpace, alphaBoolBlend {blendOption=Linear}, boneToModify {boneName=dyn_piece_35_jnt}}
   Save cached pose 'AniPieces_Grp6Cache'(Pose=Component To Local(ComponentPose=@n40)) {cacheName=AniPieces_Grp6Cache}
  ]}
 EventGraph {nodes [
   Event BlueprintUpdateAnimation()
   > n1 = Sequence()
   n2 = Sequence!(execute=@n1.then_1) -> (then_0, then_1, then_2:Exec, then_3:Exec, then_4:Exec, then_5:Exec)
   > Set piece_0(piece_0:Vector=Translate Piece!(frequency:double=2, amplitude:double=0.3, coefficient:Vector="1,0,0") -> (translateOutput:Vector=~) {is=title K2Node_MacroInstance}) -> (Output_Get:Vector)
   > Set piece_1(piece_1:Vector=Translate Piece(frequency=4, amplitude=0.4, coefficient="1.3,0.3,0.65") -> (translateOutput=~)) -> (Output_Get:Vector)
   > Set piece_2(piece_2:Vector=Translate Piece(frequency=2, amplitude=0.2, coefficient="1,3,1") -> (translateOutput=~)) -> (Output_Get:Vector)
   > Set piece_15(piece_15:Vector=Translate Piece(frequency=3, amplitude=0.4, coefficient="1,0,0") -> (translateOutput=~)) -> (Output_Get:Vector)
   > Set piece_16(piece_16:Vector=Translate Piece(frequency=3, amplitude=0.3, coefficient="0.2,0.2,0.5") -> (translateOutput=~)) -> (Output_Get:Vector)
   > Set piece_17(piece_17:Vector=Translate Piece(frequency=1, amplitude=0.4, coefficient="1,0,0") -> (translateOutput=~)) -> (Output_Get:Vector)
   > Set piece_30(piece_30:Vector=Translate Piece(frequency=1, amplitude=0.3, coefficient="1,0,0") -> (translateOutput=~)) -> (Output_Get:Vector)
   > Set piece_31(piece_31:Vector=Translate Piece(frequency=4.3, amplitude=0.5, coefficient="0.1,0.4,0.2") -> (translateOutput=~)) -> (Output_Get:Vector)
   > Set piece_32(piece_32:Vector=Translate Piece(frequency=1.2, amplitude=0.7, coefficient="1,0,0") -> (translateOutput=~)) -> (Output_Get:Vector)
   Set piece_3(execute=@n2.then_1, piece_3:Vector=Translate Piece(frequency=2.3, amplitude=0.4, coefficient="1,0,0") -> (translateOutput=~)) -> (Output_Get:Vector)
   > Set piece_4(piece_4:Vector=Translate Piece(frequency=3, amplitude=0.3, coefficient="0.3,0.54,1") -> (translateOutput=~)) -> (Output_Get:Vector)
   > Set piece_5(piece_5:Vector=Translate Piece(frequency=4, amplitude=0.4, coefficient="1,0,0") -> (translateOutput=~)) -> (Output_Get:Vector)
   > Set piece_18(piece_18:Vector=Translate Piece(frequency=2.56, amplitude=0.3, coefficient="1,0,0") -> (translateOutput=~)) -> (Output_Get:Vector)
   > Set piece_19(piece_19:Vector=Translate Piece(frequency=5, amplitude=0.5, coefficient="0.1,0.1,0.23") -> (translateOutput=~)) -> (Output_Get:Vector)
   > Set piece_20(piece_20:Vector=Translate Piece(frequency=3, amplitude=0.2, coefficient="1,0,2") -> (translateOutput=~)) -> (Output_Get:Vector)
   > Set piece_33(piece_33:Vector=Translate Piece(frequency=4.2, amplitude=0.3, coefficient="0.3,0.5,0.5") -> (translateOutput=~)) -> (Output_Get:Vector)
   > Set piece_34(piece_34:Vector=Translate Piece(frequency=1.2, amplitude=0.3, coefficient="1,2,1") -> (translateOutput=~)) -> (Output_Get:Vector)
   > Set piece_35(piece_35:Vector=Translate Piece(frequency=3, amplitude=0.3, coefficient="0.2,0.5,0.7") -> (translateOutput=~)) -> (Output_Get:Vector)
   Set piece_6(execute=@n2.then_2, piece_6:Vector=Translate Piece(frequency=1, amplitude=0.3, coefficient="1,0,0") -> (translateOutput=~)) -> (Output_Get:Vector)
   > Set piece_7(piece_7:Vector=Translate Piece(frequency=1.7, amplitude=0.4, coefficient="1,0,0") -> (translateOutput=~)) -> (Output_Get:Vector)
   > Set piece_8(piece_8:Vector=Translate Piece(frequency=2.6, amplitude=0.3, coefficient="1,1,1") -> (translateOutput=~)) -> (Output_Get:Vector)
   > Set piece_21(piece_21:Vector=Translate Piece(frequency=2.73, amplitude=0.4, coefficient="1,0,0") -> (translateOutput=~)) -> (Output_Get:Vector)
   > Set piece_22(piece_22:Vector=Translate Piece(frequency=3, amplitude=0.3, coefficient="1,1,1") -> (translateOutput=~)) -> (Output_Get:Vector)
   > Set piece_23(piece_23:Vector=Translate Piece(frequency=3.415, amplitude=0.4, coefficient="1,0,0") -> (translateOutput=~)) -> (Output_Get:Vector)
   > Set piece_36(piece_36:Vector=Translate Piece(frequency=4.3, amplitude=0.2, coefficient="1.5,0.5,1") -> (translateOutput=~)) -> (Output_Get:Vector)
   > Set piece_37(piece_37:Vector=Translate Piece(frequency=3.5, amplitude=0.3, coefficient="0.5,1,0.2") -> (translateOutput=~)) -> (Output_Get:Vector)
   > Set piece_38(piece_38:Vector=Translate Piece(frequency=4, amplitude=0.3, coefficient="0.1,0.9,0.6") -> (translateOutput=~)) -> (Output_Get:Vector)
   Set piece_9(execute=@n2.then_3, piece_9:Vector=Translate Piece(frequency=3.5, amplitude=0.4, coefficient="1,0,0") -> (translateOutput=~)) -> (Output_Get:Vector)
   > Set piece_10(piece_10:Vector=Translate Piece(frequency=2.6, amplitude=0.3, coefficient="1.2,0,1") -> (translateOutput=~)) -> (Output_Get:Vector)
   > Set piece_11(piece_11:Vector=Translate Piece(frequency=3, amplitude=0.45, coefficient="0.7,1,1.5") -> (translateOutput=~)) -> (Output_Get:Vector)
   > Set piece_24(piece_24:Vector=Translate Piece(frequency=4, amplitude=0.2, coefficient="0.5,0.3,1") -> (translateOutput=~)) -> (Output_Get:Vector)
   > Set piece_25(piece_25:Vector=Translate Piece(frequency=1.546, amplitude=0.4, coefficient="1,0,0") -> (translateOutput=~)) -> (Output_Get:Vector)
   > Set piece_26(piece_26:Vector=Translate Piece(frequency=4, amplitude=0.15, coefficient="1,1,1") -> (translateOutput=~)) -> (Output_Get:Vector)
   Set piece_12(execute=@n2.then_4, piece_12:Vector=Translate Piece(frequency=3.12, amplitude=0.3, coefficient="1,0,0") -> (translateOutput=~)) -> (Output_Get:Vector)
   > Set piece_13(piece_13:Vector=Translate Piece(frequency=3.8, amplitude=0.4, coefficient="1,0.4,1") -> (translateOutput=~)) -> (Output_Get:Vector)
   > Set piece_14(piece_14:Vector=Translate Piece(frequency=1.83, amplitude=0.3, coefficient="1,0,0") -> (translateOutput=~)) -> (Output_Get:Vector)
   > Set piece_27(piece_27:Vector=Translate Piece(frequency=1, amplitude=0.4, coefficient="1,0,0") -> (translateOutput=~)) -> (Output_Get:Vector)
   > Set piece_28(piece_28:Vector=Translate Piece(frequency=4, amplitude=0.3, coefficient="0.4,0.4,1.5") -> (translateOutput=~)) -> (Output_Get:Vector)
   > Set piece_29(piece_29:Vector=Translate Piece(frequency=3, amplitude=0.3, coefficient="0,0.4,1") -> (translateOutput=~)) -> (Output_Get:Vector)
   n3 = Gravity Override State Function DEPRECATED(
    execute=@n1.then_0
    CharacterPartAnimInstance=Self-Reference()
    GameTimeInSeconds=GetGameTimeInSeconds()
    GravityOverrideParams=Get Gravity Override Params()
   )
   > Set GravityOverride(GravityOverride=@n3)
   TryGetPawnOwner()
  ]}
 floatSinFunc {nodes [
   n1 = Inputs!() -> (time:double=~, frequency:double=~, amplitude:double=~, coefficient:double=~)
   n2 = "Sin (Radians)"(A=float * float(A=@n1, B=@n1.frequency))
   Outputs!(floatOutput:double=float * float(A=float * float(A=@n2, B=@n1.amplitude), B=@n1.coefficient))
  ]}
 translatePiece {nodes [
   n1 = Inputs!() -> (frequency=~, amplitude=~, coefficient:Vector=~)
   n2 = BreakVector(InVec=@n1.coefficient)
   n3 = GetGameTimeInSeconds()
   Outputs!(translateOutput:Vector=MakeVector(
     X=Float Sin Func!(time:double=@n3, frequency:double=@n1, amplitude:double=@n1.amplitude, coefficient:double=@n2) -> (floatOutput:double=~) {is=title K2Node_MacroInstance}
     Y=Float Sin Func(time=@n3, frequency=@n1, amplitude=@n1.amplitude, coefficient=@n2.Y) -> (floatOutput=~)
     Z=Float Sin Func(time=@n3, frequency=@n1, amplitude=@n1.amplitude, coefficient=@n2.Z) -> (floatOutput=~)
    ))
  ]}
]
variables [
 Gravity Override Params {type=GravityOverrideParamsStruct}
 GravityOverride {type=Vector}
 jumping {type=number}
 piece {type=Vector}
 piece_0 {type=Vector}
 piece_1 {type=Vector}
 piece_10 {type=Vector}
 piece_11 {type=Vector}
 piece_12 {type=Vector}
 piece_13 {type=Vector}
 piece_14 {type=Vector}
 piece_15 {type=Vector}
 piece_16 {type=Vector}
 piece_17 {type=Vector}
 piece_18 {type=Vector}
 piece_19 {type=Vector}
 piece_2 {type=Vector}
 piece_20 {type=Vector}
 piece_21 {type=Vector}
 piece_22 {type=Vector}
 piece_23 {type=Vector}
 piece_24 {type=Vector}
 piece_25 {type=Vector}
 piece_26 {type=Vector}
 piece_27 {type=Vector}
 piece_28 {type=Vector}
 piece_29 {type=Vector}
 piece_3 {type=Vector}
 piece_30 {type=Vector}
 piece_31 {type=Vector}
 piece_32 {type=Vector}
 piece_33 {type=Vector}
 piece_34 {type=Vector}
 piece_35 {type=Vector}
 piece_36 {type=Vector}
 piece_37 {type=Vector}
 piece_38 {type=Vector}
 piece_4 {type=Vector}
 piece_5 {type=Vector}
 piece_6 {type=Vector}
 piece_7 {type=Vector}
 piece_8 {type=Vector}
 piece_9 {type=Vector}
]
